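# Translucent unlit Niagara mesh-particle material for the flame hair of the Inferno character.
# World position offset combines spline thickening, a push away from texture-baked root positions and an optional Niagara-driven bend.
# Noise-distorted panning masks blend two hue-shifted emissive colors, boosted by sand sparkle outside low quality and forward/mobile shading.
# Opacity combines that mask with edge, view-angle and depth fades, plus a noisy dissolve driven by dynamic parameters.
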Material /BRCosmetics/Characters/Player/Male/Medium/Bodies/M_MED_Inferno/FX/Mat/GliderMaterial/M_Inferno_FlameHair
graph {
 properties {
  bUsedWithNiagaraMeshParticles=true
  blendMode=BLEND_Translucent
  pixelDepthOffsetMode=PDOM_Legacy
  shadingModel=MSM_Unlit
  usedShadingModels=Unlit
 }
 nodes [
  n1 = OneMinus(None=MaterialFunctionCall(Float2=TextureCoordinate()) {materialFunction=&BreakOutFloat2Components}.G)
  n2 = WorldPosition() {worldPositionShaderOffset=WPT_ExcludeAllShaderOffsets}
  n3 = NamedRerouteUsage()
  n4 = DynamicParameter() {parameterIndex=2, defaultValue=<rgba 0 0 1 1>, paramNames [Scale, Param2, Param3, Dissolve Progress]}
  n5 = Multiply(
   A=Normalize(VectorInput=Subtract(A=@n2.XYZ, B=@n3))
   B=LinearInterpolate(B=ScalarParameter() {defaultValue=10, group=WPO_FlameUp, parameterName=FlameUpStrength}, Alpha=@n4.Scale)
  )
  n6 = DynamicParameter() {desc=Parameters set by Niagara, parameterIndex=1, paramNames [X,Y,Z,WPOIntensity]}
  n7 = MaterialFunctionCall(X=@n6.X, Y=@n6.X, Z=@n6.X) {materialFunction=&MakeFloat3}
  n8 = ParticlePositionWS()
  n9 = ComponentMask(None=OneMinus(None=TextureCoordinate())) {r=true}
  n10 = ScalarParameter() {defaultValue=6, group=Interpolate WPO, parameterName=WPOBendPower}
  n11 = LinearInterpolate(
   B=Clamp(
    None=Subtract(A=@n7.Result, B=@n8)
    Min=ScalarParameter() {defaultValue=-100, group=Interpolate WPO, parameterName=Clamp_min}
    Max=ScalarParameter() {defaultValue=100, group=Interpolate WPO, parameterName=Clamp_max}
   ) {maxDefault=1000, minDefault=-1000}
   Alpha=Multiply(A=Power(Base=@n9, Exp=@n10), B=@n6.WPOIntensity)
  )
  n12 = VertexColor()
  Add(
   A=Add(A=NamedRerouteUsage(), B=Multiply(A=@n1, B=@n5))
   B=LinearInterpolate(A=Constant(), B=Multiply(A=@n11, B=@n12.G), Alpha=ScalarParameter() {group=Interpolate WPO, parameterName=Use interpolate WPO})
  )
  n13 = TextureObjectParameter() {
   group=SandA
   parameterName=SandA_Texture
   samplerType=SAMPLERTYPE_Normal
   sortPriority=0
   texture=/Game/Tools/WaterfallGenerator/Art/T_Waterfall_Sand
  }
  n14 = VectorParameter() {group=SandTexture, parameterName=SandA_UVControl, channelNames {a=VPan, b=UPan, g=VScale, r=UScale}, defaultValue=<rgba 0.5 0.4 0.1 0>}
  n15 = NamedRerouteUsage()
  n16 = Add(A=MaterialFunctionCall(Texture=@n13, AdditionalUVOffset=@n15, UVControl_RGB=@n14, UVControl_A=@n14) {materialFunction=&FXCommon_TextureSample}.R)
  n17 = TextureObjectParameter() {
   group=SandB
   parameterName=SandB_Texture
   samplerType=SAMPLERTYPE_Normal
   sortPriority=0
   texture=&T_Waterfall_Sand
  }
  n18 = VectorParameter() {group=SandTexture, parameterName=SandB_UVControl, channelNames {a=VPan, b=UPan, g=VScale, r=UScale}, defaultValue=<rgba 0.5 0.4 0.05 0>}
  n19 = Add(A=MaterialFunctionCall(Texture=@n17, AdditionalUVOffset=Multiply(A=@n15) {constB=1.2358}, UVControl_RGB=@n18, UVControl_A=@n18) {materialFunction=&FXCommon_TextureSample}.G)
  n20 = Multiply(
   A=Multiply(A=Multiply(A=@n16) {constB=0.5}, B=Multiply(A=@n19) {constB=0.5})
   B=ScalarParameter() {defaultValue=3, group=SandTexture, parameterName=MulBeforeSandRef}
  )
  n21 = VectorParameter() {group=SandTexture, parameterName=SandRefine, defaultValue=<rgba 60 5000 0 0.5>}
  n22 = VectorParameter() {group=SandTexture, parameterName=SandRefine2, defaultValue=<rgba 2 1 1 0.5>}
  n23 = TextureCoordinate()
  n24 = Power(
   Base=OneMinus(None=MaterialFunctionCall(Float2=@n23) {materialFunction=&BreakOutFloat2Components}.R)
   Exp=ScalarParameter() {defaultValue=4, group=SandTexture, parameterName=SandRefinePow}
  )
  n25 = MaterialFunctionCall(Float4=LinearInterpolate(A=AppendVector(A=@n21.RGB, B=@n21.RGB), B=AppendVector(A=@n22.RGB, B=@n22.RGB), Alpha=OneMinus(None=@n24))) {materialFunction=&BreakOutFloat4Components}
  n26 = Divide(
   A=Distance(A=CameraPositionWS(), B=WorldPosition().XYZ)
   B=ScalarParameter() {defaultValue=300, group=SandTexture, parameterName=SandRefineBoostDistance}
  ) {constB=50}
  n27 = Multiply(A=OneMinus(None=Saturate(None=@n26)), B=ScalarParameter() {defaultValue=3, group=SandTexture, parameterName=SandRefineBoostSpeed}) {constB=100}
  MaterialExpressionAppend3Vector_0 = MaterialXAppend3Vector!(A=Multiply(A=@n25.R, B=Add(A=@n27)), B=@n25.G, C=@n25.G)
  n28 = Constant()
  n29 = ShadingPathSwitch(
   Default=MaterialFunctionCall(Texture=Saturate(None=@n20), RGB=@MaterialExpressionAppend3Vector_0, A=@n25.A) {materialFunction=&MF_FX_Refine}.Result
   Forward=@n28
   Mobile=@n28
  )
  n30 = NamedRerouteUsage()
  n31 = Power(
   Base=MaterialFunctionCall(Float2=TextureCoordinate()) {materialFunction=&BreakOutFloat2Components}.R
   Exp=ScalarParameter() {defaultValue=2, group=Color, parameterName=HueShiftAreaPower}
  )
  n32 = Multiply(
   A=OneMinus(None=MaterialFunctionCall(Float2=@n30) {materialFunction=&BreakOutFloat2Components}.R)
   B=Multiply(A=OneMinus(None=@n31), B=ScalarParameter() {group=Color, parameterName=HueShiftStr})
  )
  n33 = TextureObjectParameter() {
   group=MainMask
   parameterName=MainMask_Texture
   samplerType=SAMPLERTYPE_Grayscale
   sortPriority=0
   texture=/Game/Gadgets/Assets/VinderTech_GliderChute/Glider_Thrive_Female/FX/Texture/T_RibbonBaseShape4
  }
  n34 = VectorParameter() {group=MainMask, parameterName=MainMask_UVControl, channelNames {a=VPan, b=UPan, g=VScale, r=UScale}, defaultValue=<rgba 1 1 0 0>}
  n35 = MaterialFunctionCall(Texture=@n33, AdditionalUVOffset=NamedRerouteUsage(), UVControl_RGB=@n34, UVControl_A=@n34, "[Override]TimeInput"=NamedRerouteUsage()) {materialFunction=&FXCommon_TextureSample}
  n36 = VectorParameter() {group=Color, parameterName=MaskToColorRefine, defaultValue=<rgba 15 10 0.3 0.5>}
  n37 = LinearInterpolate(
   A=VectorParameter() {group=Color, parameterName=Color, defaultValue=<rgba 0 0.0833 1 1>}.RGB
   B=VectorParameter() {group=Color, parameterName=Color2, defaultValue=<rgba 0 5.9484 10 1>}.RGB
   Alpha=MaterialFunctionCall(Texture=@n35.R, RGB=@n36.RGB, A=@n36.RGB) {materialFunction=&MF_FX_Refine}.Result
  )
  n38 = Multiply(
   A=Add(B=QualitySwitch(Default=@n29, Low=@n28)) {constA=1}
   B=Multiply(
    A=MaterialFunctionCall(Hue Shift Percentage=@n32, Texture=@n37) {materialFunction=&HueShift}.Result
    B=ScalarParameter() {defaultValue=1.5, group=Color, parameterName=EmissiveBoost}
   )
  )
  n39 = ParticleColor()
  MaterialFunctionCall(Emissive=Divide(A=Multiply(A=@n38, B=@n39.RGB)) {constB=1.977}) {materialFunction=&MF_FX_EmissiveAdjustments}
  n40 = TextureSampleParameter2D(UVs=TextureCoordinate() {coordinateIndex=1}) {
   group=WPO_BaseSetting
   parameterName=RootPosTexture
   samplerType=SAMPLERTYPE_LinearColor
   texture=/BRCosmetics/Characters/Player/Male/Medium/Bodies/M_MED_Inferno/FX/Mat/T_FireHairPos
  }
  n41 = Add(A=Multiply(A=@n40.RGB) {constB=2}) {constB=-1}
  n42 = VectorParameter() {group=WPO_BaseSetting, parameterName=BoundSize, defaultValue=<rgba 25 25 25 1>}
  NamedRerouteDeclaration(None=TransformPosition(None=Multiply(A=@n41, B=@n42.RGB)) {transformSourceType=TRANSFORMPOSSOURCE_Instance, transformType=TRANSFORMPOSSOURCE_World}) {name=RootPosWS, nodeColor=<rgba 0.3828 0 1 1>}
  NamedRerouteDeclaration(None=DynamicParameter() {parameterIndex=3, paramNames [OpacityBoost,Timer,Param3,Param4]}.Timer) {name=Timer, nodeColor=<rgba 0 0.5078 1 1>}
  n43 = VertexColor()
  n44 = Multiply(A=ScalarParameter() {defaultValue=18, group=WPO_SplineSetting, parameterName=SplineSetting_WidthB}, B=@n43.B)
  n45 = DynamicParameter() {paramNames [Param1,Param2,Param3,WidthScale]}
  n46 = ScalarParameter() {defaultValue=1, group=WPO_SplineSetting, parameterName=SplineSetting_UseDynamicInputControlTheWidth}
  n47 = Multiply(A=ScalarParameter() {defaultValue=5, group=WPO_SplineSetting, parameterName=SplineSetting_WidthT}, B=@n43.B)
  n48 = MaterialFunctionCall!(
   TangentU_WS
   TangentV_WS
   WidthBase=LinearInterpolate(A=@n44, B=Multiply(A=@n44, B=@n45.WidthScale), Alpha=@n46)
   WidthTip=LinearInterpolate(A=@n47, B=Multiply(A=@n45.WidthScale, B=@n47), Alpha=@n46)
   WorldPosition
   UVs for Projection
   UVs for Thickness
   UVs for texturing
   Expand U or V UV Channel=StaticBool()
   DeriveNormalZ
   AdditionalNormal
   AngleCorrectedNormals
   FlattenPixelNormal
   UVs For NormalShading
  ) {materialFunction=/Game/Effects/Niagara/Modules/FollowingDeform/Material/SplineThickenWithTangentInput}
  NamedRerouteDeclaration(None=@n48.WorldPositionOffset) {name=SplineWPO, nodeColor=<rgba 1 0 0.9141 1>}
  n49 = VertexColor()
  NamedRerouteDeclaration(None=MaterialFunctionCall(X=Multiply(A=@n49.R) {constB=123.7587}, Y=Multiply(A=@n49.R) {constB=12.2378}) {materialFunction=&MakeFloat2}.Result) {name=UVRandomOffset, nodeColor=<rgba 1 0.8438 0 1>}
  n50 = TextureObjectParameter() {group=NoiseA, parameterName=NoiseA_Texture, sortPriority=0, texture=/Game/Items/Traps/Textures/T_PerlinRGB_01_M}
  n51 = TextureObjectParameter() {
   group=NoiseB
   parameterName=NoiseB_Texture
   samplerType=SAMPLERTYPE_LinearGrayscale
   sortPriority=0
   texture=&T_SimplePerlin_Greyscale
  }
  n52 = NamedRerouteUsage()
  n53 = VectorParameter() {group=NoiseB, parameterName=NoiseB_UVControl, channelNames {a=VPan, b=UPan, g=VScale, r=UScale}, defaultValue=<rgba 0.7 1 0.5 0>}
  n54 = NamedRerouteUsage()
  n55 = VectorParameter() {
   group=NoiseB
   parameterName=NoiseB_NoiseControl
   sortPriority=60
   channelNames {a=VStrength, b=UStrength, g=VOffset, r=UOffset}
   defaultValue=<rgba -0.5 -0.5 0.1 0.2>
  }
  n56 = Add(
   A=NamedRerouteUsage()
   B=MaterialFunctionCall!(
    NoiseValue=MaterialFunctionCall(Texture=@n51, AdditionalUVOffset=@n52, UVControl_RGB=@n53, UVControl_A=@n53, "[Override]TimeInput"=@n54) {materialFunction=&FXCommon_TextureSample}.RG
    Control_RGB=@n55
    Control_A=@n55
   ) {materialFunction=/Game/Effects/Niagara/MasterMaterials/Functions/Common/FXCommon_NoiseControl}.NoiseValue
  )
  n57 = VectorParameter() {group=NoiseA, parameterName=NoiseA_UVControl, channelNames {a=VPan, b=UPan, g=VScale, r=UScale}, defaultValue=<rgba 0.7 1.5 0.2 0>}
  n58 = NamedRerouteUsage()
  n59 = VectorParameter() {
   group=NoiseA
   parameterName=NoiseA_NoiseControl
   sortPriority=60
   channelNames {a=VStrength, b=UStrength, g=VOffset, r=UOffset}
   defaultValue=<rgba 0 -0.2 0.8 0.2>
  }
  NamedRerouteDeclaration(None=MaterialFunctionCall(
    NoiseValue=MaterialFunctionCall(Texture=@n50, AdditionalUVOffset=@n56, UVControl_RGB=@n57, UVControl_A=@n57, "[Override]TimeInput"=@n58) {materialFunction=&FXCommon_TextureSample}.RG
    Control_RGB=@n59
    Control_A=@n59
   ) {materialFunction=&FXCommon_NoiseControl}.NoiseValue) {name=NoiseResult, nodeColor=<rgba 0 1 0.9531 1>}
  NamedRerouteDeclaration(None=@n48.Normal) {name=SplineNormal, nodeColor=<rgba 0 0.5312 1 1>}
  n60 = Abs(None=DotProduct(A=CameraVectorWS(), B=NamedRerouteUsage()))
  n61 = ScalarParameter() {defaultValue=3, group=Fresnel, parameterName=FresnelMul}
  n62 = Power(Base=Saturate(None=Multiply(A=@n60, B=@n61)), Exp=ScalarParameter() {defaultValue=1, group=Fresnel, parameterName=FresnelPow}) {constExponent=1}
  n63 = TextureCoordinate()
  n64 = VectorParameter() {
   group=Opacity
   parameterName=Opacity_EdgeMask_InvertControl
   channelNames {a="", b=Power, g=Multiply, r=Offset}
   defaultValue=<rgba 0 15 2 1>
  }
  n65 = VectorParameter() {group=Opacity, parameterName=Opacity_EdgeMask_Control, channelNames {a="", b=Power, g=Multiply, r=Offset}, defaultValue=<rgba 0 1 1 1>}
  n66 = Multiply(A=@n35.R, B=ScalarParameter() {defaultValue=1, group=Color, parameterName=OpacctyScale})
  n67 = DynamicParameter() {parameterIndex=3, paramNames [OpacityBoost,Timer,Param3,Param4]}
  n68 = Multiply(
   A=MaterialFunctionCall!(
    UorV=MaterialFunctionCall(Float2=@n63) {materialFunction=&BreakOutFloat2Components}.R
    Invert_EdgeControl=@n64
    NoInvert_EdgeControl=@n65
   ) {materialFunction=/Game/Effects/Niagara/MasterMaterials/Functions/Common/FXCommon_EdgeMask_Ex}.Mask
   B=Multiply(A=Multiply(A=@n66, B=@n67.OpacityBoost), B=VertexColor().A)
  )
  n69 = TextureCoordinate()
  n70 = Multiply(A=ComponentMask(None=@n69) {r=true}, B=OneMinus(None=ComponentMask(None=@n69) {r=true}))
  n71 = Multiply(
   A=LinearInterpolate(B=Multiply(A=@n70) {constB=2}, Alpha=ScalarParameter() {defaultValue=0.5, group=Dissolve, parameterName=Lerp Gradient Tip}) {constA=1, constAlpha=1}
   B=ComponentMask(None=TextureCoordinate()) {g=true}
  )
  n72 = Multiply(A=VertexColor().R) {constB=123.7587}
  n73 = VectorParameter() {group=Dissolve, parameterName=Dissolve Tex UV, sortPriority=6, defaultValue=<rgba 0.5 0.1 0 0>}
  n74 = AppendVector(A=@n73.R, B=@n73.R)
  n75 = TextureCoordinate()
  n76 = Add(A=AppendVector(A=@n72, B=@n72), B=Panner(Coordinate=Multiply(A=@n75, B=@n74), Speed=AppendVector(A=@n73.B, B=@n73.B)))
  n77 = LinearInterpolate(
   A=@n71
   B=Multiply(
    A=@n71
    B=TextureSampleParameter2D(UVs=@n76) {group=Dissolve, parameterName=Dissolve Tex, texture=/Game/Global/Textures/Default/Noise/T_FireNoise_01}.R
   )
   Alpha=ScalarParameter() {defaultValue=0.5, group=Dissolve, parameterName=Lerp Dissolve Tex}
  )
  n78 = ScalarParameter() {defaultValue=1, group=Dissolve, parameterName=Dissolve Power}
  n79 = ScalarParameter() {defaultValue=4, group=Dissolve, parameterName=Dissolve Edge Hardness, sliderMin=0.5}
  n80 = DynamicParameter() {parameterIndex=2, defaultValue=<rgba 0 0 1 1>, paramNames [Scale, Param2, Param3, Dissolve Progress]}
  n81 = Subtract(A=Multiply(A=Power(Base=@n77, Exp=@n78), B=@n79), B=LinearInterpolate(A=@n79, Alpha=@n80.Dissolve Progress) {constB=-1})
  n82 = Multiply(A=@n39.A, B=Multiply(A=Multiply(A=@n62, B=@n68), B=Saturate(None=@n81)))
  Saturate(None=DepthFade(Opacity=Saturate(None=@n82), FadeDistance=ScalarParameter() {defaultValue=3, parameterName=DepthFade}))
  ScalarParameter() {defaultValue=0.5, group=Dissolve, parameterName=Dissolve Dynamic Test}
  Time()
 ]
}
